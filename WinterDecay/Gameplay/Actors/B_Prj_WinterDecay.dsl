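; Winter Decay teleport projectile that spawns a teleport marker building actor (BGA) where it lands.
; It registers with the teleport weapon, updates ground-dust FX by trace timer, and resizes its capsule on hotfix-driven delays to avoid accidental hits.
; On authority, valid hits choose a marker position via surface-normal-specific traces against clipping, then spawn it, optionally attached to vehicles.
; Failed traces or water entry hide the projectile and clear the weapon's reference.
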(event Custom|RunFuseEvent
  (Utilities|Time|SetTimerbyEvent (EventDispatchers|CreateEvent :!event FuseEnded) (Variables|Default|GetFuseTime))
  (Networking|CanExecuteCosmeticEvents
    (:True
      (Audio|Components|Audio|SetSound (Variables|Default|GetGrenadeFuse_AudioComponent) (Variables|Default|GetCue_GrenadeFuseSound))
      (if (Variables|Default|GetbShowFuseSoundIndicator)
        (bind _returnvalue (Utilities|Time|SetTimerbyEvent (delegate Custom|ShowFuseIndicator) 0.3 true))
        (Variables|Default|SetFuseIndicatorTimer _returnvalue)))))

(event Custom|ShowFuseIndicator
  (Fortnite|ShowSoundIndicatorLocal :TrackedActor self :Location (Transformation|GetActorLocation) :MaxAudibleDistance (Variables|/Script/Engine.SoundBase|GetMaxDistance (Variables|Default|GetCue_GrenadeFuseSound)) :IndicatorType "Glider" :Instigator (Variables|Actor|GetInstigator) :Tint "(R=1.000000,G=1.000000,B=1.000000,A=1.000000)" :OverrideIcon (Variables|Default|GetSoundIndicatorIcon) :SoundBase (Variables|Default|GetCue_GrenadeFuseSound) :AttenuationMaxRangeMultiplier 1.0 :AttenuationIntensityMultiplier 1.0) ; Indicator for fuse
)

(event EventEndPlay
  (params (EndPlayReason :type EEndPlayReason%20Enum :default "Destroyed"))
  (|CallParentFunction EndPlayReason)
  ;[0 0.8802 0.2494 0.5] Intentionally null proj ref fed in
  (Class|BWinterDecayTeleport|NewProjectileCreatedandValids (Variables|Default|GetTeleportWeapon) 0))

(event Custom|DemonMaskVoid_SetGeneralCapsuleSize_AfterDelay
  ;[0 0.1244 0.9635 0.5] Delay then increase capsule size. This is to help prevent accidental hits when firing out car windows, water when swimming, and to help prevent general geo / terrain spots that could lead to clipping through
  (Components|Capsule|SetCapsuleSize (Variables|Collision|GetCapsuleComponent) (ScalableFloat|GetValueatLevel (Variables|Hotfix|GetHF_GeneralCapsuleSizeAfterDelay)) (ScalableFloat|GetValueatLevel (Variables|Hotfix|GetHF_GeneralCapsuleSizeAfterDelay)) true))

(event EventBeginPlay
  (bind _returnvalue (Game|GetInstigator))
  (bind _as_fort_player_pawn (Variables|Default|GetAs%20Fort%20Player%20Pawn))
  (|CallParentFunction)
  (Utilities|FlowControl|Sequence
    (:then_0
      (Collision|IgnoreActorwhenMoving (Variables|Collision|GetCapsuleComponent) _returnvalue false)
      (bind _asfort_player_pawn (Utilities|Casting|CastToFortPlayerPawn _returnvalue))
      (Variables|Default|SetAs%20Fort%20Player%20Pawn _asfort_player_pawn)
      (bind _asb_winter_decay_teleport (Utilities|Casting|CastToB_WinterDecay_Teleport (Variables|/Script/FortniteGame.FortPawn|GetCurrentWeapon _asfort_player_pawn)))
      (bind _output_get (Variables|Default|SetTeleportWeapon _asb_winter_decay_teleport))
      (Class|BWinterDecayTeleport|NewProjectileCreatedandValids _output_get self)
      (CallFunction|RunFuseEvent)
      ;[0 0.1244 0.9635 0.5] Delay then increase capsule size. This is to help prevent accidental hits when firing out car windows, water when swimming, and to help prevent general geo / terrain spots that could lead to clipping through
      (if (ScalableFloat|GetValueasBool (Variables|Hotfix|GetHF_AllowPassengerCapsuleScaling))
        (if (or (Vehicle|CanUseInventoryasDriver _as_fort_player_pawn) (Vehicle|IsPassengerinVehicle _as_fort_player_pawn))
          _
          (elif (and (WaterBody|IsSurfaceSwimming _as_fort_player_pawn) (ScalableFloat|GetValueasBool (Variables|Hotfix|GetHF_AllowSwimmingCapsuleScaling)))
            _
            (else
              (if (ScalableFloat|GetValueasBool (Variables|Hotfix|GetHF_EnableGeneralCapsuleScalingAfterDelay)) :!label _label
                (bind _returnvalue_2 (Utilities|Time|SetTimerbyEvent (delegate Custom|DemonMaskVoid_SetGeneralCapsuleSize_AfterDelay) (ScalableFloat|GetValueatLevel (Variables|Hotfix|GetHF_DelayBeforeGeneralCapsuleIncrease))))
                (Variables|Default|SetGeneralCapsuleSizeTimerHandle _returnvalue_2)))))
        (Components|Capsule|SetCapsuleSize (Variables|Collision|GetCapsuleComponent) (ScalableFloat|GetValueatLevel (Variables|Hotfix|GetHF_PassengerCapsuleScale)) (ScalableFloat|GetValueatLevel (Variables|Hotfix|GetHF_PassengerCapsuleScale)) true)
        (bind _returnvalue_3 (Utilities|Time|SetTimerbyEvent (delegate Custom|DemonMaskVoid_SetPassengerCapsuleSize_AfterDelay) (ScalableFloat|GetValueatLevel (Variables|Hotfix|GetHF_DelayBeforeCapsuleIncrease))))
        (Variables|Default|SetCapsuleSizeTimerHandle _returnvalue_3))
      (goto _label))
    (:then_1
      (Utilities|FlowControl|Sequence
        (:then_0
          ;[0.7695 0.2788 1 0.5] FX - FX Visual Feedback when Projectile is near the Ground
          (Utilities|Time|SetTimerbyEvent (delegate Custom|UpdateDustLocation) 0.15 true))))))

(event Custom|DemonMaskVoid_SetPassengerCapsuleSize_AfterDelay
  ;[0 0.1244 0.9635 0.5] Delay then increase capsule size. This is to help prevent accidental hits when firing out car windows, water when swimming, and to help prevent general geo / terrain spots that could lead to clipping through
  (Components|Capsule|SetCapsuleSize (Variables|Collision|GetCapsuleComponent) (ScalableFloat|GetValueatLevel (Variables|Hotfix|GetHF_CapsuleSizeAfterDelay)) (ScalableFloat|GetValueatLevel (Variables|Hotfix|GetHF_CapsuleSizeAfterDelay)) true)
  (if (ScalableFloat|GetValueasBool (Variables|Hotfix|GetHF_EnableGeneralCapsuleScalingAfterDelay))
    (bind _returnvalue (Utilities|Time|SetTimerbyEvent (delegate Custom|DemonMaskVoid_SetGeneralCapsuleSize_AfterDelay) (ScalableFloat|GetValueatLevel (Variables|Hotfix|GetHF_DelayBeforeGeneralCapsuleIncrease))))
    (Variables|Default|SetGeneralCapsuleSizeTimerHandle _returnvalue)))

(event Custom|UpdateDustLocation
  (bind _ns_voidmask_projectile (Variables|Default|GetNS_VoidMask_Projectile))
  ;[0.7695 0.2788 1 0.5] FX - FX Visual Feedback when Projectile is near the Ground
  (bind (_outhit _returnvalue) (Collision|LineTraceByChannel (+ (Transformation|GetActorLocation) "0.000000,0.000000,100.000000") (+ (Transformation|GetActorLocation) "0.000000,0.000000,-2000.000000")))
  (Niagara|SetNiagaraVariable(Bool) _ns_voidmask_projectile "GroundDust" _returnvalue)
  (Niagara|SetNiagaraVariable(Vector3) _ns_voidmask_projectile "DustLocation" (.Location (Collision|BreakHitResult _outhit))))

(event Custom|InitiateSpawnBGAChildEvent
  (params (HitResult :type Hit%20Result%20Structure) (HitOtherActor :type Actor%20Object%20Reference) (HitOtherComp :type Primitive%20Component%20Object%20Reference) (HitLocation :type Vector))
  (bind _hf_bga_generalnormaloffset (Variables|Hotfix|GetHF_BGA_GeneralNormalOffset))
  (bind _returnvalue (ScalableFloat|GetValueatLevel _hf_bga_generalnormaloffset))
  (bind _mindotwithup (Variables|Spawning|ValidityChecks|GetMinDotWithUp))
  (bind _hit_normal (Variables|Default|GetHit%20Normal))
  (bind _returnvalue_2 (Math|Vector|DotProduct _hit_normal "0.000000,0.000000,1.000000"))
  (bind _returnvalue_3 (> _returnvalue_2 _mindotwithup))
  (bind (_bblockinghit _binitialoverlap _time _distance _location _impactpoint _normal _impactnormal _physmat _hitactor _hitcomponent) (Collision|BreakHitResult HitResult))
  (bind _returnvalue_4 (Transformation|GetActorLocation))
  (bind _point (CallFunction|GetClosestPoint HitOtherComp _returnvalue_4))
  (bind (_asfort_aipawn _bsuccess) (Utilities|Casting|CastToFortAIPawn HitOtherActor))
  (bind _returnvalue_5 (Math|Vector|SelectVector HitLocation _point _bsuccess))
  (bind _as_fort_player_pawn (Variables|Default|GetAs%20Fort%20Player%20Pawn))
  (bind _returnvalue_6 (Actor|GetComponentbyClass _as_fort_player_pawn "/WinterDecay/Gameplay/Components/B_WinterDecay_FortPawnComponent.B_WinterDecay_FortPawnComponent_C"))
  (bind _hit_normal_2 (Variables|Default|GetHit%20Normal))
  (bind _returnvalue_7 (* _hit_normal_2 _returnvalue))
  (bind _returnvalue_8 (+ _returnvalue_5 _returnvalue_7))
  (bind _rootcomponent (Variables|Transformation|GetRootComponent HitOtherActor))
  (bind (_hit_normal_x _hit_normal_y) (Math|Vector|BreakVector (Variables|GetHit%20Normal_X)))
  (bind _lifespanafterspawnbga (Variables|Spawning|GetLifespanAfterSpawnBGA))
  (bind _hit_normal_x_2 (Math|Vector|BreakVector (Variables|GetHit%20Normal_X)))
  (bind (_hit_normal_x_3 _hit_normal_y_2 _hit_normal_z) (Math|Vector|BreakVector (Variables|GetHit%20Normal_X)))
  ;[0 0.009347 1] InitiateSpawnBGA
  ; Setup (stop simulating, do once) and then Spawn BGA on valid hit
  (Game|Components|ProjectileMovement|StopSimulating (Variables|Movement|GetProjectileMovementComponent) HitResult)
  (CallFunction|AdjustSpawnScaleforVehicles HitOtherActor)
  ;;[0.2377 0.8594 0.04905 0.5] If we're hitting ground, ensure that we don't have anything too low above us, or we'll teleport through it
  (bind _returnvalue_9 (GameplayTags|HasAnyMatchingGameplayTags(Message) HitOtherActor (Variables|Default|GetTag%20Container)))
  (if (not _returnvalue_9)
    (Variables|Default|SetActors%20to%20Ignore (Utilities|Array|MakeArray (Variables|Default|GetAs%20Fort%20Player%20Pawn)))
    (if (> _hit_normal_z 0.75)
      (bind (_outhit _returnvalue_10) (Collision|LineTraceByChannel (Transformation|GetActorLocation) (+ (Transformation|GetActorLocation) (* (Variables|Default|GetHit%20Normal) (ScalableFloat|GetValueatLevel (Variables|Hotfix|GetHF_HeightTraceForValidation)))) "TraceTypeQuery6" false (Variables|Default|GetActors%20to%20Ignore)))
      (if _returnvalue_10
        (Development|PrintString "Floor Hit Normal Failed  Trace" true true "(R=0.000000,G=0.660000,B=1.000000,A=1.000000)" 4.0)
        (CallFunction|HidePrjMeshandDelayDeathtoAllowFXToPlay)
        (Class|BWinterDecayTeleport|SetUIStateforCurrentBGAValid (Variables|Default|GetTeleportWeapon))
        (else
          (Variables|Default|SetSpawnBGAPos _returnvalue_5 :!label _label_3)))
      (elif (or (> _hit_normal_y 0.75) (< _hit_normal_y -0.75))
        (bind _spawnbga (CallFunction|CheckSideSurfaceHitNormalandValidateSpawnandPosition HitLocation _returnvalue_5 _bsuccess))
        (if (not _spawnbga)
          (CallFunction|HidePrjMeshandDelayDeathtoAllowFXToPlay)
          (Class|BWinterDecayTeleport|SetUIStateforCurrentBGAValid (Variables|Default|GetTeleportWeapon)))
        (elif _returnvalue_3
          (bind (_outhit_2 _returnvalue_11) (Collision|LineTraceByChannel (Transformation|GetActorLocation) _returnvalue_8 "TraceTypeQuery6" false (Variables|Default|GetActors%20to%20Ignore) :!label _label))
          (if _returnvalue_11
            (CallFunction|HidePrjMeshandDelayDeathtoAllowFXToPlay :!label _label_2)
            (Class|BWinterDecayTeleport|SetUIStateforCurrentBGAValid (Variables|Default|GetTeleportWeapon))
            (else
              (Variables|Default|SetSpawnBGAPos _returnvalue_8)))
          (elif (> _hit_normal_x_2 0.5)
            (goto _label)
            (else
              (bind _spawnbga_2 (CallFunction|CheckUndersideHitNormalandValidateSpawnandPosition _returnvalue_5 _returnvalue_3))
              (if (not _spawnbga_2)
                (goto _label_2)))))))
    (bind _returnvalue_12 (CallFunction|SpawnBGA _normal _hitactor _hitcomponent (Variables|Default|GetSpawnBGAPos) false HitResult))
    (bind _asbuilding_actor (Utilities|Casting|CastToBuildingActor _returnvalue_12))
    (bind _returnvalue_13 (GameplayTags|HasMatchingGameplayTag(Message) HitOtherActor "(TagName=\"Athena.Gameplay.Train.Car\")"))
    (Variables|/WinterDecay/Gameplay/Components/B_WinterDecay_FortPawnComponent.B_WinterDecay_FortPawnComponent_C|SetIsBGAOnTrain _returnvalue_13 _returnvalue_6)
    (Variables|/WinterDecay/Gameplay/Components/B_WinterDecay_FortPawnComponent.B_WinterDecay_FortPawnComponent_C|SetProjectileLocation (Transformation|GetActorLocation) _returnvalue_6)
    (Default|CallOnBGACreated _asbuilding_actor)
    (Class|BuildingActor|SetCanbeSavedtoCreativeVolume _asbuilding_actor false) ; Thrown items shouldn't get saved by default to Creative maps
    (Actor|SetLifeSpan _lifespanafterspawnbga)
    (Utilities|Time|SetTimerbyEvent (delegate Custom|DoExplodeAfterDelay) _lifespanafterspawnbga)
    ;[0.008226 1 0.1079] Allow spawning on vehicles
    (if (ScalableFloat|GetValueasBool (Variables|Hotfix|GetAttachToVehiclesHF))
      (bind _returnvalue_14 (GameplayTags|HasMatchingGameplayTag(Message) HitOtherActor (Variables|Default|GetValet%20Vehicle%20Tag%20to%20Check)))
      (if _returnvalue_14
        (bind _returnvalue_15 (Class|FortVehicleInterface|GetSkeletalMeshComponent(Message) HitOtherActor))
        (if (Transformation|DoesSocketExist _returnvalue_15 (Variables|Default|GetDriverChoiceSocket))
          (Transformation|AttachActorToComponent _returnvalue_12 _rootcomponent (Variables|Default|GetDriverChoiceSocket) "KeepWorld" "KeepWorld" "KeepWorld" true)
          (else
            _))
        (else
          ;;[1 1 1] If we hit a cow catcher
          (bind _returnvalue_16 (GameplayTags|HasMatchingGameplayTag(Message) HitOtherActor (Variables|Default|GetCow%20Catcher%20Tag%20to%20Check)))
          (if _returnvalue_16
            _
            (else
              _))))
      (Transformation|AttachActorToComponent _returnvalue_12 _rootcomponent "None" "KeepWorld" "KeepWorld" "KeepWorld" true)))
  (goto _label_3))

(event Custom|DoExplodeAfterDelay
  ;[0 0.009347 1] InitiateSpawnBGA
  ; Setup (stop simulating, do once) and then Spawn BGA on valid hit
  (if (Variables|Default|GetDoExplosionOnBGASpawn)
    (Delivery|DoExplosion)))

(event Custom|OnBGACreated_Event
  (params (BGAVoidMaskTeleportMarker :type Building%20Actor%20Object%20Reference))
  ;[0 0.009347 1] InitiateSpawnBGA
  ; Setup (stop simulating, do once) and then Spawn BGA on valid hit
  (Development|PrintString "teste"))

(event EventTick
  (params (DeltaSeconds :type Float%20%28single-precision%29)))

(event OnComponentHit(CapsuleComponent)
  (params (HitComponent :type Primitive%20Component%20Object%20Reference) (OtherActor :type Actor%20Object%20Reference) (OtherComp :type Primitive%20Component%20Object%20Reference) (NormalImpulse :type Vector :default "0, 0, 0") (Hit :type Hit%20Result%20Structure%20%28by%20ref%29))
  (Projectile|SetVelocity (Variables|Movement|GetProjectileMovementComponent) (Math|Vector|ClampVectorSize (Variables|/Script/Engine.MovementComponent|GetVelocity (Variables|Movement|GetProjectileMovementComponent)) 0.0 1.5e+03)))

(event Collision|EventHit
  (params (MyComp :type Primitive%20Component%20Object%20Reference) (Other :type Actor%20Object%20Reference) (OtherComp :type Primitive%20Component%20Object%20Reference) (bSelfMoved :type Boolean) (HitLocation :type Vector :default "0, 0, 0") (HitNormal :type Vector :default "0, 0, 0") (NormalImpulse :type Vector :default "0, 0, 0") (Hit :type Hit%20Result%20Structure%20%28by%20ref%29))
  ;[1 1 1] Event Hit; Child version duped from parent B_Prj_Athena_BGASpawner. There's enough changes to base behavior that I don't want to dirty the parent for now.
  ; Check if the projectile has hit a valid target and determine the behavior of the hit
  (Variables|Default|SetHit%20Normal HitNormal)
  (Utilities|FlowControl|Sequence
    (:then_0
      (if (not (Variables|Default|GetHasHitWater))
        (bind _inairloop (Variables|Default|GetInAirLoop)
          (:then
            (Audio|Components|Audio|FadeOut _inairloop 0.25 0.0 "Linear" :!class /Script/Engine.AudioComponent))
          (:else
            _))
        (Networking|SwitchHasAuthority
          (:Authority
            (Default|IsHitValidforSpawn OtherComp Other Hit
              (:ValidSpawn
                (CallFunction|InitiateSpawnBGAChildEvent Hit Other OtherComp HitLocation)))))))
    (:then_1
      (Niagara|SetNiagaraVariable(Bool) (Variables|Default|GetNS_VoidMask_Projectile) "ProjStopped" true))))

(event EventOnStop
  (params (Hit :type Hit%20Result%20Structure%20%28by%20ref%29))
  (|CallParentFunction Hit))

(event EventOnExploded
  (params (HitActors :type Array%20of%20Actor%20Object%20References) (HitResults :type Array%20of%20Hit%20Result%20Structures))
  (CallFunction|DelayedExplode))

(event Custom|DelayedExplode
  (params (Hit :type Hit%20Result%20Structure))
  (bind (_bblockinghit _binitialoverlap _time _distance _location _impactpoint _normal _impactnormal _physmat _hitactor) (Collision|BreakHitResult Hit))
  (if (not (Variables|Default|GetHasExploded))
    (Variables|Default|SetHasExploded true)
    (GameplayCue|ExecuteGameplayCue%28Burst%29Local self (Variables|Default|GetImpactExplodeGC) (Ability|GameplayCue|MakeGameplayCueParameters :NormalizedMagnitude 0.0 :RawMagnitude 256.0 :Location _impactpoint :Normal _normal :EffectCauser _hitactor :PhysicalMaterial _physmat :GameplayEffectLevel 1 :AbilityLevel 1 :bReplicateLocationWhenUsingMinimalRepProxy false))
    (CallFunction|HidePrjMeshandDelayDeathtoAllowFXToPlay)
    ;[0 0.3094 0.04008] The delay prevents us stopping the InAirLoop before the Explosion SFX have played.
    ; NOTE: Can't set delays in a function, but can't set breakpoints in a Macro, so just sharing this logic
    (Utilities|FlowControl|Delay 0.2)
    (Audio|Components|Audio|FadeOut (Variables|Default|GetInAirLoop) 0.25 0.0 "Linear" :!class /Script/Engine.AudioComponent)))

(event OnEnterWaterEvent(WaterInteractionComponent)
  (params (WaterBody :type Fort%20Water%20Body%20Actor%20Object%20Reference) (WaterInteractionComponent :type Fort%20Water%20Interaction%20Component%20Object%20Reference) (bIsFirstBody :type Boolean))
  ;[1 1 1] Duping from parent B_Prj_Athena_BGASpawner. We let this 'live' a little past normal to allow FX to play / fade out
  (Variables|Default|SetHasHitWater true)
  (if bIsFirstBody
    (CallFunction|PlaySplashVFX)
    (Actor|SetLifeSpan (Variables|Spawning|GetLifespanAfterSpawnBGA))
    (Collision|SetActorEnableCollision false)
    (CallFunction|HidePrjMeshandDelayDeathtoAllowFXToPlay)
    ;[0 0.3094 0.04008] The delay prevents us stopping the InAirLoop before the Explosion SFX have played.
    ; NOTE: Can't set delays in a function, but can't set breakpoints in a Macro, so just sharing this logic
    (Utilities|FlowControl|Delay 0.2)
    (Audio|Components|Audio|FadeOut (Variables|Default|GetInAirLoop) 0.25 0.0 "Linear" :!class /Script/Engine.AudioComponent)))

(event Custom|PlaySplashVFX
  (Effects|Components|ParticleSystem|SpawnFXAtLocation "/Game/Effects/Niagara/Systems/Enviroment/Water/NS_WaterSplash_Medium.NS_WaterSplash_Medium" (Transformation|GetActorLocation)))

(fn AdjustSpawnScaleForVehicles
  (params (Hit%20Actor :type Actor%20Object%20Reference))
  (bind _returnvalue (GameplayTags|HasMatchingGameplayTag(Message) Hit%20Actor (Variables|Default|GetValet%20Vehicle%20Tag%20to%20Check)))
  (if _returnvalue
    (bind _returnvalue_2 (GameplayTags|HasMatchingGameplayTag(Message) Hit%20Actor (Variables|Default|GetBike%20Tag%20to%20Check)))
    (if _returnvalue_2
      (Variables|Spawning|TransformData|SetSpawnScale (Math|Conversions|ToVector(Float) (ScalableFloat|GetValueatLevel (Variables|Hotfix|GetAttachToBikeScaleHF))))
      (else
        (Variables|Spawning|TransformData|SetSpawnScale (Math|Conversions|ToVector(Float) (ScalableFloat|GetValueatLevel (Variables|Hotfix|GetAttachToVehicleScaleHF))))))
    (return)))

(fn CheckSideSurfaceHitNormalAndValidateSpawnAndPosition
  (params (HitLocation :type Vector) (HitLocationForAI :type Vector) (FortAiPawn :type Boolean))
  (bind _hf_bga_normaloffsetwhenwallhitfromvehicle (Variables|Hotfix|GetHF_BGA_NormalOffsetWhenWallHitFromVehicle))
  (bind _returnvalue (ScalableFloat|GetValueatLevel _hf_bga_normaloffsetwhenwallhitfromvehicle))
  (bind _hf_bga_normaloffsetwhenwallhit (Variables|Hotfix|GetHF_BGA_NormalOffsetWhenWallHit))
  (bind _returnvalue_2 (ScalableFloat|GetValueatLevel _hf_bga_normaloffsetwhenwallhit))
  (bind _capsulecomponent (Variables|Collision|GetCapsuleComponent))
  (bind _capsuleradius (Variables|/Script/Engine.CapsuleComponent|GetCapsuleRadius _capsulecomponent))
  (bind _returnvalue_3 (< _capsuleradius 24.0))
  (bind _returnvalue_4 (Math|Float|SelectFloat _returnvalue _returnvalue_2 _returnvalue_3))
  (bind _hit_normal (Variables|Default|GetHit%20Normal))
  (bind _returnvalue_5 (* _hit_normal _returnvalue_4))
  (bind _returnvalue_6 (+ _returnvalue_5 HitLocation))
  (bind _returnvalue_7 (Math|Vector|SelectVector HitLocationForAI _returnvalue_6 FortAiPawn))
  ;[0.004603 0.7552 0.04243 0.5] If we're hitting a vertical or side wall type object, move the BGA away a bit to avoid teleporting through walls
  (bind (_outhit _returnvalue_8) (Collision|LineTraceByChannel (+ (Transformation|GetActorLocation) (* (Variables|Default|GetHit%20Normal) (Math|Float|SelectFloat 24.0 0.0 (Vehicle|IsinVehicle (Variables|Default|GetAs%20Fort%20Player%20Pawn))))) (+ (Transformation|GetActorLocation) _returnvalue_5) "TraceTypeQuery6" false (Variables|Default|GetActors%20to%20Ignore)))
  (if _returnvalue_8
    (Development|PrintString "Side Hit Normal Failed  Trace" true true "(R=0.000000,G=0.660000,B=1.000000,A=1.000000)" 4.0)
    (return false)
    (else
      (bind (_outhit_2 _returnvalue_9) (Collision|LineTraceByChannel (Transformation|GetActorLocation) (+ (Transformation|GetActorLocation) (* (Math|Vector|VectorUp) 4e+01)) "TraceTypeQuery6" false (Variables|Default|GetActors%20to%20Ignore)))
      (if _returnvalue_9
        (Variables|Default|SetSpawnBGAPos (Math|Vector|SelectVector (+ (Transformation|GetActorLocation) (* (Math|Vector|VectorUp) -3e+01)) _returnvalue_7 _returnvalue_9))
        (else
          (bind (_outhit_3 _returnvalue_10) (Collision|LineTraceByChannel (Transformation|GetActorLocation) (+ (Transformation|GetActorLocation) (* (Math|Vector|VectorUp) 6e+01)) "TraceTypeQuery6" false (Variables|Default|GetActors%20to%20Ignore)))
          (if _returnvalue_10
            _)
          (Variables|Default|SetSpawnBGAPos (Math|Vector|SelectVector (+ (Transformation|GetActorLocation) (* (Math|Vector|VectorUp) -5e+01)) _returnvalue_7 _returnvalue_10))))
      (return true))))

(fn CheckUndersideHitNormalAndValidateSpawnAndPosition
  (params (InputPin :type Vector) (NewParam :type Boolean))
  (bind (_outhit _returnvalue) (Collision|LineTraceByChannel (+ (Transformation|GetActorLocation) 0) (+ (Transformation|GetActorLocation) (* (Variables|Default|GetHit%20Normal) (ScalableFloat|GetValueatLevel (Variables|Hotfix|GetHF_BGA_TraceSizeForUndersideHits)))) "TraceTypeQuery1" false (Variables|Default|GetActors%20to%20Ignore)))
  (if _returnvalue
    (Development|PrintString "Underside Hit Normal Failed  Trace" true true "(R=0.000000,G=0.660000,B=1.000000,A=1.000000)" 4.0)
    (return false)
    (else
      (Variables|Default|SetSpawnBGAPos (Math|Vector|SelectVector InputPin (Utilities|Operators|Subtract InputPin (Math|Vector|MakeVector 0.0 0.0 (ScalableFloat|GetValueatLevel (Variables|Hotfix|GetHF_BGAZOffsetDownWhenHIttingUnder)))) NewParam))
      (return true))))

(fn FuseEnded)

(fn HidePrjMeshAndDelayDeathToAllowFXToPlay
  (bind _as_fort_player_pawn (Variables|Default|GetAs%20Fort%20Player%20Pawn))
  (bind _returnvalue (Actor|GetComponentbyClass _as_fort_player_pawn "/DemonMaskVoid/Gameplay/Components/B_DemonMask_Void_FortPawnComponent.B_DemonMask_Void_FortPawnComponent_C"))
  ;[1 1 1] Parent class BGA Spawner sets lifespan to 2s & SetActor Hidden by default on BGA spawn. Exposed those settings to childen, so this BP determines it's own lifetime and sets mesh hidden to allow FX to finish properly
  (Components|Movement|StopMovementImmediately (Variables|Components|GetRotatingMovement))
  (Rendering|SetVisibility (Variables|Components|GetTeleportMarkerMesh) false false)
  ;[0 0.2373 0.8177 0.5] We're going to destroy this, null the reference so we don't let players start teleporting to it, then we wind up failing it for them
  (Class|BWinterDecayTeleport|NewProjectileCreatedandValids (Variables|Default|GetTeleportWeapon) 0)
  (if (Utilities|IsValid _returnvalue :!class /Script/Engine.KismetSystemLibrary)
    (Class|FortPawnComponentVoidTeleport|ServerSetCurrentTrackedVoidProjectile _returnvalue 0)))

(fn RevertToPickupAfterNFailedAttempts
  (if (ScalableFloat|GetValueasBool (Variables|Hotfix|GetShouldRevertToPickupOnFailedAttemptsHF))
    (bind _output_get (Variables|Default|SetFailedSpawnAttempts (+ (Variables|Default|GetFailedSpawnAttempts) 1)))
    (if (> _output_get (ScalableFloat|GetValueasInteger (Variables|Hotfix|GetNumFailedAttemptsCheckHF)))
      _)))

(fn ShouldBounceOffActor
  (params (Hit%20Actor :type Actor%20Object%20Reference) (Hit%20Component :type Primitive%20Component%20Object%20Reference))
  (bind (_asbuilding_prop _bsuccess) (Utilities|Casting|CastToBuildingProp Hit%20Actor))
  (bind (_asfort_athena_supply_drop _bsuccess_2) (Utilities|Casting|CastToFortAthenaSupplyDrop Hit%20Actor))
  (bind (_asfort_projectile_base _bsuccess_3) (Utilities|Casting|CastToFortProjectileBase Hit%20Actor))
  (bind (_asathena_supply_drop_balloon _bsuccess_4) (Utilities|Casting|CastToAthenaSupplyDropBalloon Hit%20Actor))
  (bind (_asfort_pawn _bsuccess_5) (Utilities|Casting|CastToFortPawn Hit%20Actor))
  (Utilities|IsValid Hit%20Actor
    (:Is%20Valid
      (Utilities|IsValid Hit%20Component
        (:Is%20Valid
          (return (Math|Boolean|ORBoolean (Utilities|Array|ContainsItem (Variables|/Script/Engine.ActorComponent|GetComponentTags Hit%20Component) (LocalVariables|GetBounce%20Tag)) (Utilities|Array|ContainsItem (Variables|Actor|GetTags Hit%20Actor) (LocalVariables|GetBounce%20Tag)) _bsuccess _bsuccess_5 _bsuccess_2 _bsuccess_4 _bsuccess_3)))
        (:Is%20Not%20Valid
          _)))
    (:Is%20Not%20Valid
      _))
  (return false))

(fn ConstructionScript
  (|CallParentFunction)
  (Components|Movement|SetUpdatedComponent (Variables|Components|GetRotatingMovement) (Variables|Components|GetTeleportMarkerMesh)))
